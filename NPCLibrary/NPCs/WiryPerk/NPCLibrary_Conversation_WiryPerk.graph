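# Conversation database for the WiryPerk NPC, defining the greeting and top-level branches when a player talks to them.
# The entry tag leads to one Speech node whose default line can be replaced by contextual lines for first-ever meetings, per-match conversation count, or specific player character cosmetics.
# Priorities rank these lines, with first-meeting priority read from a curve row.
# The greeting then branches through link nodes to shared Quests, BasicServices, and WiryPerk's own Dialogue entry.
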ConversationDatabase /NPCLibrary/NPCs/WiryPerk/NPCLibrary_Conversation_WiryPerk
entry_points [{tag=Conversation.Entry.WiryPerk, destinations [#1]}]
nodes [
 ConversationEntryPointNode_7 {
  id=#1
  class=ConversationEntryPointNode
  outputs [#2]
  properties {entryTag {tagName=Conversation.Entry.WiryPerk}}
 }
 Speech_C_4 {
  id=#2
  class=Speech_C
  outputs [#3, #4, #5]
  properties {
   nodeName=""
   generalConfig {
    defaultMessage="This Bug war is blowin' up, and NOT in the fun way!"
    contextualMessages [
     {
      message="Oh hiiiiii!"
      contextRequirements [{requirement {class=FortControllerRequirement_IsFirstEverConversationWithNPC}}]
      priority {curve {rowName=Tandem.Speech.Priority.FirstEverInteraction}}
     }
     {
      message="This Bug war is blowin' up, and NOT in the fun way!"
      contextRequirements [{requirement {class=FortControllerRequirement_CheckNPCMatchConversationCount, properties {nPCUniqueIdTag {tagName=AISpawnerData.Type.Tandem.WiryPerk}}}}]
      priority {curve {curveTable=null, rowName=None}}
     }
     {
      message=uWu?
      contextRequirements [{requirement {class=FortControllerRequirement_CheckNPCMatchConversationCount, properties {nPCUniqueIdTag {tagName=AISpawnerData.Type.Tandem.WiryPerk}}}}]
      priority {curve {curveTable=null, rowName=None}}
     }
     {
      message="I told you. We're not calling my combat gloves the \"Bear Arms.\""
      contextRequirements [{requirement {class=FortControllerRequirement_HasCID, properties {allowedCIds [/BRCosmetics/Athena/Items/Cosmetics/Characters/Character_ClumsyChewLaw]}}}]
      priority {value=2, curve {curveTable=null, rowName=None}}
     }
     {
      message=You like the suit? I just upgraded your interior cooling array.
      contextRequirements [{requirement {class=FortControllerRequirement_HasCID, properties {allowedCIds [/BRCosmetics/Athena/Items/Cosmetics/Characters/Character_TalonPane_Flail]}}}]
      priority {value=2, curve {curveTable=null, rowName=None}}
     }
    ]
   }
  }
 }
 ConversationLinkNode_2 {id=#5, class=ConversationLinkNode, properties {remoteEntryTag {tagName=Conversation.Entry.WiryPerk.Dialogue}}}
 ConversationLinkNode_0 {id=#4, class=ConversationLinkNode, properties {remoteEntryTag {tagName=Conversation.Entry.BasicServices}}}
 ConversationLinkNode_1 {id=#3, class=ConversationLinkNode, properties {remoteEntryTag {tagName=Conversation.Entry.NPCLibrary.Quests}}}
]
