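# Conversation database for the WiryPerk NPC's dialogue, an optional "Ask about..." branch where the character gives flavor lore.
# The entry point leads to a greeting speech that offers three topics (Mae's Criminal Past, the O-M-Gloves, O.X.R.) plus a Back choice.
# Each topic plays a linear chain of speech nodes, and all three converge on a shared ReturnToStart node labeled BACK.
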ConversationDatabase /NPCLibrary/NPCs/WiryPerk/NPCLibrary_Conversation_WiryPerk_Dialogue
entry_points [{tag=Conversation.Entry.WiryPerk.Dialogue, destinations [#1]}]
nodes [
 ConversationEntryPointNode_7 {
  id=#1
  class=ConversationEntryPointNode
  outputs [#2]
  properties {entryTag {tagName=Conversation.Entry.WiryPerk.Dialogue}}
 }
 Speech_C_0 {
  id=#2
  class=Speech_C
  outputs [#3, #4, #5, #6]
  properties {nodeName="", softTaskIcon=&T_UI_NPC_question, generalConfig {defaultMessage=Yeaaaaah?}}
  sub_nodes [ConversationChoiceNode_4 {class=ConversationChoiceNode, properties {defaultChoiceDisplayText=Ask about...}}]
 }
 Speech_C_6 {
  id=#6
  class=Speech_C
  outputs [#7]
  properties {
   nodeName=""
   softTaskIcon=&T_UI_NPC_question
   generalConfig {defaultMessage="They're the real deal. Totally fighting for a good cause. "}
  }
  sub_nodes [ConversationChoiceNode_9 {class=ConversationChoiceNode, properties {defaultChoiceDisplayText=O.X.R.}}]
 }
 Speech_C_8 {
  id=#7
  class=Speech_C
  outputs [#8]
  properties {nodeName="", generalConfig {defaultMessage="For someone who's spent her whole career running with criminals, it's... an adjustment."}}
 }
 Speech_C_9 {
  id=#8
  class=Speech_C
  outputs [#9]
  properties {nodeName="", generalConfig {defaultMessage=Someone tried to talk to me about ETHICS and CONSEQUENCES the other day??}}
 }
 Speech_C_10 {
  id=#9
  class=Speech_C
  outputs [#10]
  properties {nodeName="", generalConfig {defaultMessage="Like, ew. What am I supposed to do with that? WORRY? GROSS!!"}}
 }
 ReturnToStart_C_23 {
  id=#10
  class=ReturnToStart_C
  properties {softTaskIcon=&T_UI_NPC_Back}
  sub_nodes [ConversationChoiceNode_6 {class=ConversationChoiceNode, properties {defaultChoiceDisplayText=BACK}}]
 }
 Speech_C_2 {
  id=#5
  class=Speech_C
  outputs [#11]
  properties {
   nodeName=""
   softTaskIcon=&T_UI_NPC_question
   generalConfig {defaultMessage="They may look like toys, and function like toys..."}
  }
  sub_nodes [ConversationChoiceNode_8 {class=ConversationChoiceNode, properties {defaultChoiceDisplayText=The O-M-Gloves}}]
 }
 Speech_C_3 {
  id=#11
  class=Speech_C
  outputs [#12]
  properties {nodeName="", generalConfig {defaultMessage="But the O-M-Gloves are SERIOUS weapons, okay? "}}
 }
 Speech_C_1 {
  id=#12
  class=Speech_C
  outputs [#13]
  properties {nodeName="", generalConfig {defaultMessage="Ultra-dangerous, ultra-fun-to-use, kind-of-gamified weapons."}}
 }
 Speech_C_11 {
  id=#13
  class=Speech_C
  outputs [#10]
  properties {nodeName="", generalConfig {defaultMessage="Point is, be careful with these things! You could hurt yourself by having too much fun!!"}}
 }
 Back_C_1 {
  id=#4
  class=Back_C
  sub_nodes [ConversationChoiceNode_7 {class=ConversationChoiceNode, properties {defaultChoiceDisplayText=BACK}}]
 }
 Speech_C_7 {
  id=#3
  class=Speech_C
  outputs [#14]
  properties {
   nodeName=""
   softTaskIcon=&T_UI_NPC_question
   generalConfig {defaultMessage="Look, hackers-slash-explosives-experts only have two jobs available to us."}
  }
  sub_nodes [ConversationChoiceNode_5 {class=ConversationChoiceNode, properties {defaultChoiceDisplayText=Mae's Criminal Past}}]
 }
 Speech_C_15 {
  id=#14
  class=Speech_C
  outputs [#15]
  properties {nodeName="", generalConfig {defaultMessage="One: a life of crime. Been there, done that, blew up, got too notorious, etcetera :D"}}
 }
 Speech_C_14 {
  id=#15
  class=Speech_C
  outputs [#16]
  properties {nodeName="", generalConfig {defaultMessage="And two: weapons development for shadowy pseudo-military organizations. "}}
 }
 Speech_C_5 {
  id=#16
  class=Speech_C
  outputs [#10]
  properties {nodeName="", generalConfig {defaultMessage="If O.X.R. wants to pay me a hilarious amount of Bars to do what I love, who am I to say no?"}}
 }
]
